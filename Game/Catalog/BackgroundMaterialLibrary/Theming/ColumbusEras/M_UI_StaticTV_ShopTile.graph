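# Translucent UI material that draws an animated static-TV background for item shop tiles, themed for ColumbusEras.
# Three panned noise samples build the static, with a timed step periodically switching its UV scale and distortion for bigger glitches.
# Scrolling scanlines, sine-driven glitch bands and a rotated gradient are built alongside, using color dodge blends and tile background colors.
# Rounded-mask SDF supplies the shape and outline, taking its corner radius from the shop parameter collection.
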Material /Game/Catalog/BackgroundMaterialLibrary/Theming/ColumbusEras/M_UI_StaticTV_ShopTile
graph {
 properties {bAutomaticallySetUsageInEditor=false, blendMode=BLEND_Translucent, materialDomain=MD_UI, usedShadingModels=Default Lit}
 nodes [
  n1 = NamedRerouteUsage()
  n2 = Multiply(A=LinearInterpolate(A=NamedRerouteUsage(), B=Min(A=@n1)))
  n3 = NamedRerouteUsage()
  n4 = Min(A=Power(Base=Multiply(A=@n2, B=@n3)) {constExponent=0.5}) {constB=0.99}
  n5 = NamedRerouteUsage()
  n6 = NamedRerouteUsage()
  n7 = NamedRerouteUsage()
  LinearInterpolate(
   A=LinearInterpolate(A=@n4, B=@n5, Alpha=Multiply(A=@n6) {constB=0.3})
   B=@n5
   Alpha=Multiply(A=@n7, B=ScalarParameter() {defaultValue=0.6, parameterName=Outline Opacity})
  )
  n8 = NamedRerouteUsage()
  n9 = NamedRerouteUsage()
  LinearInterpolate(
   A=NamedRerouteUsage()
   B=MaterialFunctionCall!(Blend=@n4, Base=LinearInterpolate(B=@n8, Alpha=@n9)) {materialFunction=/Engine/Functions/Engine_MaterialFunctions03/Blends/Blend_ColorDodge}.Result
   Alpha=@n7
  )
  MaterialFunctionCall!(Input Alpha=Constant() {r=1}, bOverrideCornerRadius, Corner Radius Override) {materialFunction=/FortUILibrary/Systems/ItemShopTile/Materials/MaterialFunctions/MF_UI_ItemShop_RoundedCorners}
  n10 = LinearInterpolate(Alpha=NamedRerouteUsage()) {constA=0.25}
  n11 = Multiply(A=@n10, B=NamedRerouteUsage())
  Multiply(A=@n11)
  NamedRerouteDeclaration(None=MaterialFunctionCall() {materialFunction=&MF_UI_UVScaling}.Scaled to Fit) {name=UVScaling_Fit, nodeColor=<rgba 0.0781 0 1 1>}
  NamedRerouteDeclaration(None=NamedRerouteUsage()) {name=Color2, nodeColor=<rgba 1 0.0007 0.2048 1>}
  n12 = ScalarParameter() {defaultValue=0.344, group=Static, parameterName=StaticLowValue}
  n13 = ScalarParameter() {defaultValue=1, group=Static, parameterName=StaticHighValue}
  n14 = TextureCoordinate()
  n15 = Multiply(A=Time(), B=ScalarParameter() {defaultValue=0.08, parameterName=BigGlitchSpeed})
  n16 = Round(None=Power(Base=Frac(None=@n15)))
  n17 = LinearInterpolate(A=@n14, B=Multiply(A=@n14) {constB=0.75}, Alpha=@n16)
  n18 = TextureSample(UVs=Panner(Coordinate=@n17) {speedX=2.5, speedY=0.05}) {texture=&/T_StaticNoise}
  n19 = Multiply(
   A=TextureSample(UVs=Panner(Coordinate=@n17) {speedX=3, speedY=-0.025}) {texture=&T_StaticNoise}.G
   B=TextureSample(UVs=Panner(Coordinate=@n17) {speedX=-1.5, speedY=-0.015}) {texture=&T_StaticNoise}.R
  )
  n20 = Multiply(
   A=Add(A=Multiply(A=@n18.B) {constB=0.1}, B=@n19)
   B=LinearInterpolate(
    A=ScalarParameter() {defaultValue=1, group=Static, parameterName=Static Small Scale}
    B=ScalarParameter() {defaultValue=10, group=Static, parameterName=Static Big Scale}
   )
  )
  n21 = LinearInterpolate(
   A=ScalarParameter() {defaultValue=5, group=Static, parameterName=Static UV Distort}
   B=ScalarParameter() {defaultValue=5, group=Static, parameterName=Static UV Big Distort}
   Alpha=@n16
  )
  n22 = Multiply(A=Add(A=LinearInterpolate(A=@n12, B=@n13, Alpha=@n20)), B=AppendVector(A=Multiply(A=@n19, B=@n21), B=Constant() {r=3}))
  NamedRerouteDeclaration(None=Multiply(A=LinearInterpolate(Alpha=@n22), B=ScalarParameter() {defaultValue=0.5, parameterName=Static Opacity}) {constB=0.8}) {name=Hologram Noise, nodeColor=<rgba 0 0 0 1>}
  NamedRerouteDeclaration(None=@n10) {name=BG, nodeColor=<rgba 0 0 0 1>}
  n23 = MaterialFunctionCall() {materialFunction=/FortUILibrary/Systems/ItemShopTile/Materials/MaterialFunctions/MF_UI_ItemShop_BackgroundParams}
  NamedRerouteDeclaration(None=@n23.Color 2) {name=Secondary Color, nodeColor=<rgba 0 0 0 1>}
  NamedRerouteDeclaration(None=@n23.Color 3) {name=Accent Color, nodeColor=<rgba 0 0 0 1>}
  NamedRerouteDeclaration(None=@n11) {name=Flash Tire 01 - Glow, nodeColor=<rgba 0 0 0 1>}
  n24 = NamedRerouteUsage()
  n25 = NamedRerouteUsage()
  n26 = MaterialFunctionCall(
   UVs=NamedRerouteUsage()
   GradientStart=ScalarParameter() {parameterName=Gradient Start}
   GradientEnd=ScalarParameter() {defaultValue=0.74, parameterName=Gradient End}
   RotationAngle=ScalarParameter() {defaultValue=-0.15, parameterName=GradientRotation}
  ) {materialFunction=&MF_UI_LinearGradient}
  NamedRerouteDeclaration(None=MaterialFunctionCall(Blend=NamedRerouteUsage(), Base=LinearInterpolate(A=@n24, B=@n25, Alpha=@n26.Result)) {materialFunction=&Blend_ColorDodge}.Result) {name=Flash Tire 01 - Gradient, nodeColor=<rgba 0 0 0 1>}
  NamedRerouteDeclaration(None=NamedRerouteUsage()) {name=Color3, nodeColor=<rgba 0 1 0.625 1>}
  n27 = CollectionParameter() {collection=/FortUILibrary/Systems/ItemShopTile/MPC_ItemShop, parameterName=CornerRadius}
  n28 = MaterialFunctionCall(
   Corner Radius=MaterialFunctionCall(R=@n27, G=@n27, B=@n27, A=@n27) {materialFunction=&AppendMany}.RGBA
   Radius in Pixels=Constant() {r=1}
   Outline Thickness=ScalarParameter() {defaultValue=6, parameterName=Outline Thickness}
  ) {materialFunction=&MF_UI_CoreUI_RoundedMask}
  n29 = OneMinus(None=SmoothStep(Value=@n28.Outline SDF) {constMax=0.1})
  NamedRerouteDeclaration(None=LinearInterpolate(A=Multiply(A=@n29) {constB=0.35}, Alpha=@n28.Outline Mask)) {name=OutlineFrame, nodeColor=<rgba 0 0.5547 1 1>}
  n30 = Panner(Coordinate=NamedRerouteUsage(), Time=Multiply(A=Time(), B=ScalarParameter() {defaultValue=0.1, parameterName=Glitch Speed}) {constB=0.5}) {speedY=0.5}
  n31 = ComponentMask(None=@n30) {g=true}
  n32 = Multiply(A=@n31) {constB=3.4}
  n33 = Multiply(A=ComponentMask(None=@n30) {g=true}) {constB=2.2}
  n34 = Multiply(A=Frac(None=Multiply(A=@n31) {constB=10}), B=Min(A=Sine(None=@n32), B=Sine(None=@n33)))
  NamedRerouteDeclaration(None=Multiply(A=Power(Base=@n34), B=ScalarParameter() {defaultValue=0.9, parameterName=Glitch Opacity}) {constB=0.5}) {name=Glitch, nodeColor=<rgba 0 0 0 1>}
  n35 = ScalarParameter() {defaultValue=0.6, parameterName=Scanline - Opacity}
  n36 = ScalarParameter() {defaultValue=0.6, parameterName=Scanline - Thickness}
  n37 = Time()
  n38 = ScalarParameter() {defaultValue=0.008, parameterName=Scanline - Speed}
  n39 = ComponentMask(None=Panner(Coordinate=NamedRerouteUsage(), Time=Multiply(A=@n37, B=@n38)) {speedY=1}) {g=true}
  n40 = ScalarParameter() {defaultValue=150, parameterName=Scanline - Amount}
  n41 = SmoothStep(Min=Saturate(None=Subtract(A=@n36) {constB=0.001}), Max=@n36, Value=Frac(None=Multiply(A=@n39, B=@n40)))
  NamedRerouteDeclaration(None=Add(A=Multiply(A=@n35, B=@n41), B=NamedRerouteUsage())) {name=Scanline, nodeColor=<rgba 0 0 0 1>}
  NamedRerouteDeclaration(None=@n28.Shape Mask) {name=FlatBGMask, nodeColor=<rgba 0 0.7187 1 1>}
  NamedRerouteDeclaration(None=MaterialFunctionCall!(ScreenUV, TextureSize, ScalePivotPoint) {materialFunction=/FortUILibrary/MaterialFunctions/Basics/MF_UI_ScreenSpaceUVScaling}.Scaled to Fit) {name=SSUV_Fit, nodeColor=<rgba 1 0 0.2812 1>}
  NamedRerouteDeclaration(None=@n23.Color 1) {name=Primary Color, nodeColor=<rgba 0 0 0 1>}
  NamedRerouteDeclaration(None=NamedRerouteUsage()) {name=Color1, nodeColor=<rgba 1 0.0007 0.2048 1>}
 ]
}
